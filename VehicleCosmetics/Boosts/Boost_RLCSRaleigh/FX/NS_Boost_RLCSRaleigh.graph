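# Niagara system for the RLCS Raleigh car boost cosmetic, drawing warm-colored ribbon trails from the boost ports.
# One emitter spawns ribbons per port at skeletal positions offset by the car offset, pushing them backward with curve-faded inherited source movement.
# Four ribbon renderers share the particles, with curves animating three of four layer colors over lifetime.
# The painted-variant override is disabled, but the painted flag still feeds a per-particle vector.
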NiagaraSystem /VehicleCosmetics/Boosts/Boost_RLCSRaleigh/FX/NS_Boost_RLCSRaleigh
dependencies {
 usedDataInterfaces [&NiagaraDataInterfaceCurve, &NiagaraDataInterfaceSkeletalMesh]
 usedDynamicInputs []
 usedModules [
  &self:SystemSpawnScript.SystemScriptSource.SystemScriptGraph.NiagaraNodeAssignment_0.SetVariables_#1
  &SystemState
 ]
}
emitters {NE_VC_Boost_Ribbon_Parent001 {
  emitterUpdateScript [
   &EmitterState()
   &VC_FilteredSocketCount()
   &VC_RateQualityHandler()
   &VC_SpawnRibbon_v1(Number of Ribbons=$Emitter.PortCount, SpawnRate=$Emitter.Spawn Rate)
  ]
  particleSpawnScript [
   &VC_PaintedVariantsOverride() {enabled=false}
   &VC_InitializeRibbon_v1(
    Custom Offset=true
    Enable Skeletal Position=true
    Lifetime=1
    Local Vector Offset=&AddVectorToPosition(Position=<0 0 0>, Vector=$User.DTCarOffset)
    Offset Amount=1
    Ribbon Width=130
   )
   &AddVelocity(Velocity=<-1400 0 0>) {module=AddVelocity001}
   /SetVariables(Particles.WidthTag0=90, Particles.WidthTag2=65, Particles.WidthTag3=80) {isSetParameters=true}
  ]
  particleUpdateScript [
   &ParticleState()
   &InheritSourceMovement(
    Applied Position Scale=&Multiply_VectorByFloat(
     Vector=<1 1 1>
     Float=&FloatFromCurve(FloatCurve=DataInterface() {propertyValues {Curve {keys [{interpMode=RCIM_Cubic, tangentMode=RCTM_Break, time=0.3, value=1}, {interpMode=RCIM_Cubic, tangentMode=RCTM_Break, time=1, value=0}]}}})
    )
    ApplyForce=false
    ApplyPosition=true
    ApplyPositionFalloff=false
   )
   &SolveForcesAndVelocity(Clamp Velocity=false, Limit Acceleration=false)
   &ScaleColor(
    Scale Alpha=&FloatFromCurve(FloatCurve=DataInterface() {propertyValues {Curve {keys [<at 0 0 RCIM_Cubic>, {interpMode=RCIM_Cubic, tangentMode=RCTM_User, time=0.35, value=2}, <at 1 0>]}}})
    Scale RGB=&VectorFromFloat(Value=&FloatFromCurve(FloatCurve=DataInterface() {propertyValues {Curve {
         defaultValue=1
         keys [
          {interpMode=RCIM_Cubic, tangentMode=RCTM_User, time=0.2, value=0.5}
          {interpMode=RCIM_Cubic, tangentMode=RCTM_User, time=0.5, value=3}
          {tangentMode=RCTM_User, time=0.7, value=0.5}
         ]
        }}}))
    ScaleA=true
    ScaleRGB=true
   )
   &VC_UpdateRibbon_v1(Enable Skeletal Position=true)
   /SetVariables(
    Particles.Index3:Vector4f=&MakeVector4(X=&MakeCustomFloatFromBool(Bool=$User.bIsPainted, True Float=1, False Float=0), Y=1, Z=0, W=1)
    Particles.RibbonColor0:LinearColor=&VC_DynamicScaleColor(
     Color to Scale:LinearColor=<rgba 1 0.1413 0 1>
     RGB Scale:Vector3f=&VectorFromFloat(Value=&FloatFromCurve(FloatCurve=DataInterface() {propertyValues {Curve {
          defaultValue=1
          keys [
           {interpMode=RCIM_Cubic, tangentMode=RCTM_User, time=0.2, value=1.2}
           {interpMode=RCIM_Cubic, tangentMode=RCTM_User, time=0.5, value=1.4}
           {tangentMode=RCTM_User, time=0.7, value=0.94}
          ]
         }}}))
     Alpha Scale:NiagaraFloat=&FloatFromCurve(
      FloatCurve=DataInterface() {propertyValues {Curve {keys [
          {arriveTangent=10.7641, interpMode=RCIM_Cubic, leaveTangent=10.7641, tangentMode=RCTM_User, time=0, value=0}
          {interpMode=RCIM_Cubic, tangentMode=RCTM_User, time=0.35, value=2}
          {arriveTangent=-0.286, tangentMode=RCTM_User, time=1, value=0}
         ]}}}
      Scale Curve=0.3
     )
    )
    Particles.RibbonColor1:LinearColor=&VC_DynamicScaleColor(
     Color to Scale:LinearColor=<rgba 2 0.1198 0.0417 1>
     RGB Scale:Vector3f=&VectorFromFloat(Value=&FloatFromCurve(FloatCurve=DataInterface() {propertyValues {Curve {
          defaultValue=1
          keys [
           {interpMode=RCIM_Cubic, tangentMode=RCTM_User, time=0.2, value=0.5}
           {interpMode=RCIM_Cubic, tangentMode=RCTM_User, time=0.5, value=1}
           {tangentMode=RCTM_User, time=1, value=0.7}
          ]
         }}}))
     Alpha Scale:NiagaraFloat=&FloatFromCurve(
      FloatCurve=DataInterface() {propertyValues {Curve {keys [<at 0.2 1 RCIM_Cubic>, {arriveTangent=-3.1954, interpMode=RCIM_Cubic, leaveTangent=-3.1954, tangentMode=RCTM_User, time=0.8, value=0}]}}}
      Scale Curve=0.05
     )
    )
    Particles.RibbonColor2:LinearColor=&VC_DynamicScaleColor(
     Color to Scale:LinearColor=<rgba 0.7083 0.0424 0.0148 0.9826>
     RGB Scale:Vector3f=&VectorFromFloat(Value=&FloatFromCurve(FloatCurve=DataInterface() {propertyValues {Curve {keys [<at 0.2 0.5 RCIM_Cubic>,<at 0.5 1 RCIM_Cubic>,<at 0.7 0.5 RCIM_Cubic>]}}}))
     Alpha Scale:NiagaraFloat=&FloatFromCurve(
      FloatCurve=DataInterface() {propertyValues {Curve {
         defaultValue=1
         keys [
          {arriveTangent=-0.0405, interpMode=RCIM_Cubic, leaveTangent=-0.0405, tangentMode=RCTM_User, time=0.2, value=1}
          {arriveTangent=-3.6149, interpMode=RCIM_Cubic, leaveTangent=-3.6149, tangentMode=RCTM_User, time=0.8, value=0}
         ]
        }}}
      Scale Curve=0.05
     )
    )
    Particles.RibbonColor3:LinearColor=<rgba 1.4156 0.3539 0.4424 1>
   ) {isSetParameters=true}
  ]
  rendererClasses [&NiagaraRibbonRendererProperties]
  renderers [
   {rendererClass=&NiagaraRibbonRendererProperties}
   {rendererClass=&NiagaraRibbonRendererProperties, rendererIndex=1}
   {rendererClass=&NiagaraRibbonRendererProperties, rendererIndex=2}
   {rendererClass=&NiagaraRibbonRendererProperties, rendererIndex=3}
  ]
 }}
summary {
 systemName=NS_Boost_RLCSRaleigh
 emitters [{emitterName=NE_VC_Boost_Ribbon_Parent001, rendererClasses [&NiagaraRibbonRendererProperties]}]
 userVariables [
  User.bIsFrontendPreview {type=NiagaraBool}
  User.bIsPainted {type=NiagaraBool}
  User.ColorVariantID {type=NiagaraInt32}
  User.DTCarOffset {type=Vector3f, description="", defaultValue {struct=Vector3f, value=<0 0 0>}}
  User.PaintedColor {type=LinearColor, description="", defaultValue {struct=LinearColor, value=<rgba 1 1 1 1>}}
  User.PaintedScalar {type=NiagaraFloat, defaultValue=1, description=""}
  User.PortCount {type=NiagaraInt32, defaultValue=1}
 ]
}
